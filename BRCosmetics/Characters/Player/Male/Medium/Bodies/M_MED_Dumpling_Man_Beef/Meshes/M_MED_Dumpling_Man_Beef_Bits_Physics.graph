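# Physics asset for the accessory bits of the Dumpling Man Beef player cosmetic body, covering a rolling pin, shaker and steaks.
# The steaks form a two-box chain under a small master capsule, which links to the pelvis with free rotation.
# Rolling pin and shaker capsules hang from the pelvis with angular limits, much tighter on the pin.
# A left-thigh capsule, also rotationally free on the pelvis, likely serves as a collider for the dangling pieces.
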PhysicsAsset /BRCosmetics/Characters/Player/Male/Medium/Bodies/M_MED_Dumpling_Man_Beef/Meshes/M_MED_Dumpling_Man_Beef_Bits_Physics
bodies {
 dyn_rollingpin [{shapeType=Capsule, length=5, radius=4, center=<2.3503 0 0>, rotation=<rot 90 0 0>}]
 dyn_shaker [{shapeType=Capsule, length=5.7078, radius=4.1767, center=<2.8238 0 0>, rotation=<rot 90 0 0>}]
 dyn_steak_01 [{shapeType=Box, extentX=15.5061, extentY=4.0131, extentZ=15.2175, center=<7.5918 0 0>, rotation=<rot 0 0 90>}]
 dyn_steak_02 [{shapeType=Box, extentX=8.3911, extentY=2.5043, extentZ=16.1435, center=<6.6053 0 0>, rotation=<rot 0 0 -90>}]
 dyn_steak_MASTER [{shapeType=Capsule, length=0.505, radius=0.505, rotation=<rot 0 -90 0>}]
 pelvis [{shapeType=Capsule, length=1, radius=1}]
 thigh_l [{shapeType=Capsule, length=13.9363, radius=11.595, center=<-17.832 1.5059 0.0779>, rotation=<rot 90 0 0>}]
}
constraints [
 {bone1Name=dyn_steak_01, bone2Name=dyn_steak_MASTER, swing1LimitDegrees=45, swing2LimitDegrees=45, twistLimitDegrees=5}
 {bone1Name=dyn_steak_02, bone2Name=dyn_steak_01, swing1LimitDegrees=45, swing2LimitDegrees=45, twistLimitDegrees=5}
 {
  bone1Name=thigh_l
  bone2Name=pelvis
  swing1LimitDegrees=45
  swing1Motion=Free
  swing2LimitDegrees=45
  swing2Motion=Free
  twistLimitDegrees=45
  twistMotion=Free
 }
 {
  bone1Name=dyn_steak_MASTER
  bone2Name=pelvis
  swing1LimitDegrees=45
  swing1Motion=Free
  swing2LimitDegrees=45
  swing2Motion=Free
  twistLimitDegrees=45
  twistMotion=Free
 }
 {bone1Name=dyn_rollingpin, bone2Name=pelvis, swing1LimitDegrees=20, swing2LimitDegrees=1, twistLimitDegrees=5}
 {bone1Name=dyn_shaker, bone2Name=pelvis, swing1LimitDegrees=45, swing2LimitDegrees=45, twistLimitDegrees=5}
]
