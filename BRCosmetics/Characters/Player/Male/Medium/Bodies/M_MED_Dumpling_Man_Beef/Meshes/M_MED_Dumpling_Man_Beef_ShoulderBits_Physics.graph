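# Physics asset for the shoulder-strap pieces of the male medium Dumpling Man Beef player body cosmetic, defining their collision and joint limits.
# Capsules on both clavicles provide shoulder collision volumes, with small boxes on each side's strap master bone and larger boxes on the dynamic strap bones.
# On each side, two two-bone chains (A and B) are constrained to their master, allowing moderate swing and little twist.
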PhysicsAsset /BRCosmetics/Characters/Player/Male/Medium/Bodies/M_MED_Dumpling_Man_Beef/Meshes/M_MED_Dumpling_Man_Beef_ShoulderBits_Physics
bodies {
 clavicle_l [{shapeType=Capsule, length=15.0557, radius=10.5377, center=<2.7266 -0.4671 -2.1322>, rotation=<rot 79.9997 0 0>}]
 clavicle_r [{shapeType=Capsule, length=13.3222, radius=10.6906, center=<-1.7443 0.1825 2.1721>, rotation=<rot 79.9997 0 0>}]
 dyn_shoulder_strap_A_01_l [{shapeType=Box, extentX=5, extentY=4.5, extentZ=2, center=<2.5 0 0>}]
 dyn_shoulder_strap_A_01_r [{shapeType=Box, extentX=5, extentY=4.5, extentZ=2, center=<2.5 0 0>}]
 dyn_shoulder_strap_A_02_l [{shapeType=Box, extentX=5, extentY=4.5, extentZ=2, center=<2.5 0 0>}]
 dyn_shoulder_strap_A_02_r [{shapeType=Box, extentX=5, extentY=4.5, extentZ=2, center=<2.5 0 0>}]
 dyn_shoulder_strap_B_01_l [{shapeType=Box, extentX=5, extentY=4.5, extentZ=2, center=<2.5 0 0>}]
 dyn_shoulder_strap_B_01_r [{shapeType=Box, extentX=5, extentY=4.5, extentZ=2, center=<2.5 0 0>}]
 dyn_shoulder_strap_B_02_l [{shapeType=Box, extentX=5, extentY=4.5, extentZ=2, center=<2.5 0 0>}]
 dyn_shoulder_strap_B_02_r [{shapeType=Box, extentX=5, extentY=4.5, extentZ=2, center=<2.5 0 0>}]
 dyn_shoulder_straps_master_l [{shapeType=Box, extentX=2, extentY=2, extentZ=2}]
 dyn_shoulder_straps_master_r [{shapeType=Box, extentX=2, extentY=2, extentZ=2}]
}
constraints [
 {
  bone1Name=dyn_shoulder_strap_B_01_l
  bone2Name=dyn_shoulder_straps_master_l
  swing1LimitDegrees=45
  swing2LimitDegrees=45
  twistLimitDegrees=5
 }
 {bone1Name=dyn_shoulder_strap_B_02_l, bone2Name=dyn_shoulder_strap_B_01_l, swing1LimitDegrees=45, swing2LimitDegrees=45, twistLimitDegrees=5}
 {
  bone1Name=dyn_shoulder_strap_A_01_l
  bone2Name=dyn_shoulder_straps_master_l
  swing1LimitDegrees=45
  swing2LimitDegrees=45
  twistLimitDegrees=5
 }
 {bone1Name=dyn_shoulder_strap_A_02_l, bone2Name=dyn_shoulder_strap_A_01_l, swing1LimitDegrees=45, swing2LimitDegrees=45, twistLimitDegrees=5}
 {
  bone1Name=dyn_shoulder_strap_A_01_r
  bone2Name=dyn_shoulder_straps_master_r
  swing1LimitDegrees=45
  swing2LimitDegrees=45
  twistLimitDegrees=5
 }
 {bone1Name=dyn_shoulder_strap_A_02_r, bone2Name=dyn_shoulder_strap_A_01_r, swing1LimitDegrees=45, swing2LimitDegrees=45, twistLimitDegrees=5}
 {
  bone1Name=dyn_shoulder_strap_B_01_r
  bone2Name=dyn_shoulder_straps_master_r
  swing1LimitDegrees=45
  swing2LimitDegrees=45
  twistLimitDegrees=5
 }
 {bone1Name=dyn_shoulder_strap_B_02_r, bone2Name=dyn_shoulder_strap_B_01_r, swing1LimitDegrees=45, swing2LimitDegrees=45, twistLimitDegrees=5}
]
