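# Post-process style AnimBP adding secondary motion to the Dumpling Man Beef body cosmetic's straps, skirt and hanging food and utensils.
# The attached parent's pose feeds a Control Rig, then a series of AnimDynamics, constraints, Look At and additive bone offsets.
# Left-thigh yaw updates crouch_alpha for utensil and steak adjustments; crouch events also set it and its inverse, which weights the steak chain.
# Both RigidBody nodes and several older nodes are unconnected.
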AnimBlueprint /BRCosmetics/Characters/Player/Male/Medium/Bodies/M_MED_Dumpling_Man_Beef/Meshes/M_MED_Dumpling_Man_Beef_AnimBp
target_skeleton=&/M_MED_Dumpling_Man_Beef_Skeleton
graphs [
 AnimGraph {nodes [
   n1 = Anim Dynamics - Bone: dyn_shoulder_strap_A_01_l(Alpha=1, ExternalForce="0,0,0") {
    angularDampingOverride=0.999
    bGravityOverrideInSimSpace=true
    bOverrideAngularDamping=true
    simulationSpace=RootRelative
    alphaBoolBlend {blendOption=Linear}
    boundBone {boneName=dyn_shoulder_strap_A_01_l}
    physicsBodyDefinitions [{
      sphereCollisionRadius=0
      boundBone {boneName=dyn_shoulder_strap_A_01_l}
      boxExtents=<10 10 2>
      constraintSetup {angularLimitsMax=<0 5 10>, angularLimitsMin=<0 -30 -10>, angularTarget=<0 0 0>}
      localJointOffset=<2 0 0>
     }]
   }
   n2 = Anim Dynamics - Bone: dyn_shoulder_strap_A_02_l(ComponentPose=@n1, Alpha=1, ExternalForce="0,0,0") {
    angularDampingOverride=0.999
    bGravityOverrideInSimSpace=true
    bOverrideAngularDamping=true
    simulationSpace=RootRelative
    alphaBoolBlend {blendOption=Linear}
    boundBone {boneName=dyn_shoulder_strap_A_02_l}
    physicsBodyDefinitions [{
      sphereCollisionRadius=0
      boundBone {boneName=dyn_shoulder_strap_A_02_l}
      boxExtents=<10 10 2>
      constraintSetup {angularLimitsMax=<0 5 10>, angularLimitsMin=<0 -30 -10>, angularTarget=<0 0 0>}
      localJointOffset=<2 0 0>
     }]
   }
   n3 = Anim Dynamics - Bone: dyn_shoulder_strap_B_01_l(ComponentPose=@n2, Alpha=1, ExternalForce="0,0,0") {
    angularDampingOverride=0.999
    bGravityOverrideInSimSpace=true
    bOverrideAngularDamping=true
    simulationSpace=RootRelative
    alphaBoolBlend {blendOption=Linear}
    boundBone {boneName=dyn_shoulder_strap_B_01_l}
    physicsBodyDefinitions [{
      sphereCollisionRadius=0
      boundBone {boneName=dyn_shoulder_strap_B_01_l}
      boxExtents=<10 10 2>
      constraintSetup {angularLimitsMax=<0 0 10>, angularLimitsMin=<0 -30 0>, angularTarget=<0 0 0>}
      localJointOffset=<2 0 0>
     }]
   }
   Anim Dynamics - Bone: dyn_shoulder_strap_B_02_l(ComponentPose=@n3, Alpha=1, ExternalForce="0,0,0") {
    angularDampingOverride=0.999
    bGravityOverrideInSimSpace=true
    bOverrideAngularDamping=true
    simulationSpace=RootRelative
    alphaBoolBlend {blendOption=Linear}
    boundBone {boneName=dyn_shoulder_strap_B_02_l}
    physicsBodyDefinitions [{
      sphereCollisionRadius=0
      boundBone {boneName=dyn_shoulder_strap_B_02_l}
      boxExtents=<10 10 2>
      constraintSetup {angularLimitsMax=<0 0 10>, angularLimitsMin=<0 -30 0>, angularTarget=<0 0 0>}
      localJointOffset=<2 0 0>
     }]
   }
   Anim Dynamics - Bone: dyn_steak_02(Alpha=1, ExternalForce="0,0,0") {
    angularDampingOverride=0.99
    angularSpringConstant=1500
    bAngularSpring=true
    bGravityOverrideInSimSpace=true
    bOverrideAngularDamping=true
    bUsePlanarLimit=false
    gravityScale=0.35
    alphaBoolBlend {blendOption=Linear}
    alphaScaleBiasClamp {bInterpResult=true}
    boundBone {boneName=dyn_steak_02}
    physicsBodyDefinitions [{
      sphereCollisionRadius=0
      boundBone {boneName=dyn_steak_02}
      boxExtents=<5 20 20>
      constraintSetup {angularLimitsMax=<0 10 10>, angularLimitsMin=<0 -20 -10>, angularTarget=<0 0 0>}
      localJointOffset=<5 0 0>
     }]
   }
   Anim Dynamics - Bone: dyn_shoulder_strap_A_02_r(Alpha=0, ExternalForce="0,0,0") {
    angularDampingOverride=0.999
    bGravityOverrideInSimSpace=true
    bOverrideAngularDamping=true
    gravityScale=0.35
    simulationSpace=RootRelative
    alphaBoolBlend {blendOption=Linear}
    boundBone {boneName=dyn_shoulder_strap_A_02_r}
    componentAppliedLinearAccClamp=<10 10 10>
    componentLinearAccScale {y=10}
    componentLinearVelScale {y=10}
    physicsBodyDefinitions [{
      sphereCollisionRadius=0
      boundBone {boneName=dyn_shoulder_strap_A_02_r}
      boxExtents=<1 10 2>
      constraintSetup {angularLimitsMax=<0 10 10>, angularLimitsMin=<0 -30 -10>, angularTarget=<0 0 0>}
      localJointOffset=<1 0 0>
     }]
   }
   Anim Dynamics - Bone: dyn_shoulder_strap_B_02_r(Alpha=1, ExternalForce="0,0,0") {
    angularDampingOverride=0.999
    bGravityOverrideInSimSpace=true
    bOverrideAngularDamping=true
    gravityScale=0.35
    simulationSpace=RootRelative
    alphaBoolBlend {blendOption=Linear}
    boundBone {boneName=dyn_shoulder_strap_B_02_r}
    componentAppliedLinearAccClamp=<10 10 10>
    componentLinearAccScale {y=10}
    componentLinearVelScale {y=10}
    physicsBodyDefinitions [{
      sphereCollisionRadius=0
      boundBone {boneName=dyn_shoulder_strap_B_02_r}
      boxExtents=<10 10 2>
      constraintSetup {angularLimitsMax=<0 10 10>, angularLimitsMin=<0 -30 0>, angularTarget=<0 0 0>}
      localJointOffset=<5 0 0>
     }]
   }
   "dyn_spoon.rotateY = thigh_l.rotateZ - Bone Driven Controller"!(ComponentPose:Component Space Pose Link Structure, Alpha:float=1) -> (Pose:Component Space Pose Link Structure) {
    is=title AnimGraphNode_BoneDrivenController
    bAffectTargetRotationY=true
    sourceComponent=RotationZ
    alphaBoolBlend {blendOption=Linear}
    sourceBone {boneName=thigh_l}
    targetBone {boneName=dyn_spoon}
   }
   "Transform (Modify) Bone - Bone: dyn_steak_MASTER"(Alpha=Get crouch_invert_alpha() -> (crouch_invert_alpha:double), Translation="0,0,0", Rotation="0,0,0", Scale="1,1,1") {rotationMode=BMM_Additive, alphaBoolBlend {blendOption=Linear}, boneToModify {boneName=dyn_steak_MASTER}}
   RigidBody() {
    bClampLinearTranslationLimitToRefPose=true
    bOverrideWorldGravity=true
    overridePhysicsAsset=&/M_MED_Dumpling_Man_Beef_ShoulderBits_Physics
    simulationSpace=BaseBoneSpace
    alphaBoolBlend {blendOption=Linear}
    alphaScaleBiasClamp {bInterpResult=true}
    baseBoneRef {boneName=spine_05}
    componentAppliedLinearAccClamp=<10 10 10>
    componentLinearAccScale=<10 10 10>
    componentLinearVelScale=<10 10 10>
    overrideWorldGravity {z=-400}
   }
   RigidBody(Alpha=Get crouch_invert_alpha()) {
    bOverrideWorldGravity=true
    overridePhysicsAsset=&/M_MED_Dumpling_Man_Beef_Bits_Physics
    simulationSpace=BaseBoneSpace
    alphaBoolBlend {blendOption=Linear}
    alphaScaleBiasClamp {bInterpResult=true}
    baseBoneRef {boneName=pelvis}
    componentAppliedLinearAccClamp=<10 10 10>
    overrideWorldGravity {z=-980}
   }
   n4 = Control Rig(Source=Copy Pose From Mesh() {bUseAttachedParent=true}) {
    alphaBoolBlend {blendOption=Linear}
    controlRigAssetReference {blueprintRigClass=&/M_MED_Dumpling_Man_Beef_CtrlRig.M_MED_Dumpling_Man_Beef_CtrlRig_C}
   }
   n5 = Anim Dynamics - Bone: dyn_skirt_front_01_l(ComponentPose=Local To Component(LocalPose=@n4), Alpha=0.1, ExternalForce="0,0,0") {
    angularDampingOverride=0.999
    bGravityOverrideInSimSpace=true
    bOverrideAngularDamping=true
    gravityScale=0.35
    simulationSpace=RootRelative
    alphaBoolBlend {blendOption=Linear}
    boundBone {boneName=dyn_skirt_front_01_l}
    physicsBodyDefinitions [{
      sphereCollisionRadius=0
      boundBone {boneName=dyn_skirt_front_01_l}
      boxExtents=<5 20 20>
      constraintSetup {angularLimitsMax=<0 30 0>, angularLimitsMin=<0 0 0>, angularTarget=<0 0 0>}
      localJointOffset=<5 0 0>
     }]
   }
   n6 = Anim Dynamics - Bone: dyn_skirt_front_02_l(ComponentPose=@n5, Alpha=0.3, ExternalForce="0,0,0") {
    angularDampingOverride=0.999
    bGravityOverrideInSimSpace=true
    bOverrideAngularDamping=true
    gravityScale=0.35
    simulationSpace=RootRelative
    alphaBoolBlend {blendOption=Linear}
    boundBone {boneName=dyn_skirt_front_02_l}
    physicsBodyDefinitions [{
      sphereCollisionRadius=0
      boundBone {boneName=dyn_skirt_front_02_l}
      boxExtents=<5 20 20>
      constraintSetup {angularLimitsMax=<0 30 0>, angularLimitsMin=<0 0 0>, angularTarget=<0 0 0>}
      localJointOffset=<5 0 0>
     }]
   }
   n7 = Anim Dynamics - Bone: dyn_skirt_side_01_l(ComponentPose=@n6, Alpha=1, ExternalForce="0,0,0") {
    angularDampingOverride=0.999
    bGravityOverrideInSimSpace=true
    bOverrideAngularDamping=true
    gravityScale=0.35
    simulationSpace=RootRelative
    alphaBoolBlend {blendOption=Linear}
    boundBone {boneName=dyn_skirt_side_01_l}
    physicsBodyDefinitions [{
      sphereCollisionRadius=0
      boundBone {boneName=dyn_skirt_side_01_l}
      boxExtents=<5 20 20>
      constraintSetup {angularLimitsMax=<0 10 10>, angularLimitsMin=<0 0 0>, angularTarget=<0 0 0>}
      localJointOffset=<5 0 0>
     }]
   }
   n8 = Anim Dynamics - Bone: dyn_skirt_side_02_l(ComponentPose=@n7, Alpha=1, ExternalForce="0,0,0") {
    angularDampingOverride=0.999
    bGravityOverrideInSimSpace=true
    bOverrideAngularDamping=true
    gravityScale=0.35
    simulationSpace=RootRelative
    alphaBoolBlend {blendOption=Linear}
    boundBone {boneName=dyn_skirt_side_02_l}
    physicsBodyDefinitions [{
      sphereCollisionRadius=0
      boundBone {boneName=dyn_skirt_side_02_l}
      boxExtents=<5 20 20>
      constraintSetup {angularLimitsMax=<0 10 10>, angularLimitsMin=<0 0 0>, angularTarget=<0 0 0>}
      localJointOffset=<5 0 0>
     }]
   }
   n9 = Anim Dynamics - Bone: dyn_skirt_mid_01_l(ComponentPose=@n8, Alpha=1, ExternalForce="0,0,0") {
    angularDampingOverride=0.999
    bGravityOverrideInSimSpace=true
    bOverrideAngularDamping=true
    gravityScale=0.35
    simulationSpace=RootRelative
    alphaBoolBlend {blendOption=Linear}
    boundBone {boneName=dyn_skirt_mid_01_l}
    physicsBodyDefinitions [{
      sphereCollisionRadius=0
      boundBone {boneName=dyn_skirt_mid_01_l}
      boxExtents=<5 20 20>
      constraintSetup {angularLimitsMax=<0 30 0>, angularLimitsMin=<0 0 0>, angularTarget=<0 0 0>}
      localJointOffset=<5 0 0>
     }]
   }
   n10 = Anim Dynamics - Bone: dyn_skirt_mid_02_l(ComponentPose=@n9, Alpha=1, ExternalForce="0,0,0") {
    angularDampingOverride=0.999
    bGravityOverrideInSimSpace=true
    bOverrideAngularDamping=true
    gravityScale=0.35
    simulationSpace=RootRelative
    alphaBoolBlend {blendOption=Linear}
    boundBone {boneName=dyn_skirt_mid_02_l}
    physicsBodyDefinitions [{
      sphereCollisionRadius=0
      boundBone {boneName=dyn_skirt_mid_02_l}
      boxExtents=<5 20 20>
      constraintSetup {angularLimitsMax=<0 30 0>, angularLimitsMin=<0 0 0>, angularTarget=<0 0 0>}
      localJointOffset=<5 0 0>
     }]
   }
   n11 = Anim Dynamics - Bone: dyn_skirt_front_01_r(ComponentPose=@n10, Alpha=1, ExternalForce="0,0,0") {
    angularDampingOverride=0.999
    bGravityOverrideInSimSpace=true
    bOverrideAngularDamping=true
    gravityScale=0.35
    simulationSpace=RootRelative
    alphaBoolBlend {blendOption=Linear}
    boundBone {boneName=dyn_skirt_front_01_r}
    physicsBodyDefinitions [{
      sphereCollisionRadius=0
      boundBone {boneName=dyn_skirt_front_01_r}
      boxExtents=<5 20 20>
      constraintSetup {angularLimitsMax=<0 30 0>, angularLimitsMin=<0 0 0>, angularTarget=<0 0 0>}
      localJointOffset=<5 0 0>
     }]
   }
   n12 = Anim Dynamics - Bone: dyn_skirt_front_02_r(ComponentPose=@n11, Alpha=1, ExternalForce="0,0,0") {
    angularDampingOverride=0.999
    bGravityOverrideInSimSpace=true
    bOverrideAngularDamping=true
    gravityScale=0.35
    simulationSpace=RootRelative
    alphaBoolBlend {blendOption=Linear}
    boundBone {boneName=dyn_skirt_front_02_r}
    physicsBodyDefinitions [{
      sphereCollisionRadius=0
      boundBone {boneName=dyn_skirt_front_02_r}
      boxExtents=<5 20 20>
      constraintSetup {angularLimitsMax=<0 30 0>, angularLimitsMin=<0 0 0>, angularTarget=<0 0 0>}
      localJointOffset=<5 0 0>
     }]
   }
   n13 = Anim Dynamics - Bone: dyn_skirt_side_01_r(ComponentPose=@n12, Alpha=1, ExternalForce="0,0,0") {
    angularDampingOverride=0.999
    bGravityOverrideInSimSpace=true
    bOverrideAngularDamping=true
    gravityScale=0.35
    simulationSpace=RootRelative
    alphaBoolBlend {blendOption=Linear}
    boundBone {boneName=dyn_skirt_side_01_r}
    physicsBodyDefinitions [{
      sphereCollisionRadius=0
      boundBone {boneName=dyn_skirt_side_01_r}
      boxExtents=<5 20 20>
      constraintSetup {angularLimitsMax=<0 30 0>, angularLimitsMin=<0 0 0>, angularTarget=<0 0 0>}
      localJointOffset=<5 0 0>
     }]
   }
   n14 = Anim Dynamics - Bone: dyn_skirt_side_02_r(ComponentPose=@n13, Alpha=1, ExternalForce="0,0,0") {
    angularDampingOverride=0.999
    bGravityOverrideInSimSpace=true
    bOverrideAngularDamping=true
    gravityScale=0.35
    simulationSpace=RootRelative
    alphaBoolBlend {blendOption=Linear}
    boundBone {boneName=dyn_skirt_side_02_r}
    physicsBodyDefinitions [{
      sphereCollisionRadius=0
      boundBone {boneName=dyn_skirt_side_02_r}
      boxExtents=<5 20 20>
      constraintSetup {angularLimitsMax=<0 30 0>, angularLimitsMin=<0 0 0>, angularTarget=<0 0 0>}
      localJointOffset=<5 0 0>
     }]
   }
   n15 = Anim Dynamics - Bone: dyn_skirt_mid_01_r(ComponentPose=@n14, Alpha=1, ExternalForce="0,0,0") {
    angularDampingOverride=0.999
    bGravityOverrideInSimSpace=true
    bOverrideAngularDamping=true
    gravityScale=0.35
    simulationSpace=RootRelative
    alphaBoolBlend {blendOption=Linear}
    boundBone {boneName=dyn_skirt_mid_01_r}
    physicsBodyDefinitions [{
      sphereCollisionRadius=0
      boundBone {boneName=dyn_skirt_mid_01_r}
      boxExtents=<5 20 20>
      constraintSetup {angularLimitsMax=<0 30 0>, angularLimitsMin=<0 0 0>, angularTarget=<0 0 0>}
      localJointOffset=<5 0 0>
     }]
   }
   n16 = Anim Dynamics - Bone: dyn_skirt_mid_02_r(ComponentPose=@n15, Alpha=1, ExternalForce="0,0,0") {
    angularDampingOverride=0.999
    bGravityOverrideInSimSpace=true
    bOverrideAngularDamping=true
    gravityScale=0.35
    simulationSpace=RootRelative
    alphaBoolBlend {blendOption=Linear}
    boundBone {boneName=dyn_skirt_mid_02_r}
    physicsBodyDefinitions [{
      sphereCollisionRadius=0
      boundBone {boneName=dyn_skirt_mid_02_r}
      boxExtents=<5 20 20>
      constraintSetup {angularLimitsMax=<0 30 0>, angularLimitsMin=<0 0 0>, angularTarget=<0 0 0>}
      localJointOffset=<5 0 0>
     }]
   }
   n17 = "Constraint - dyn_skirt_back_01 (PPP)"!(ComponentPose=@n16, Alpha=0, ConstraintWeights_0=0.1, ConstraintWeights_1:float=0.1, ConstraintWeights_2:float=0.8) -> (Pose) {
    alphaBoolBlend {blendOption=Linear}
    boneToModify {boneName=dyn_skirt_back_01}
    constraintSetup [{targetBone {boneName=thigh_twist_01_l}}, {targetBone {boneName=thigh_twist_01_r}}, {targetBone {boneName=pelvis}}]
    constraintWeights [1,1,1]
   }
   n18 = Anim Dynamics - Bone: dyn_skirt_back_01(ComponentPose=@n17, Alpha=1, ExternalForce="0,0,0") {
    angularDampingOverride=0.999
    bGravityOverrideInSimSpace=true
    bOverrideAngularDamping=true
    gravityScale=0.35
    simulationSpace=RootRelative
    alphaBoolBlend {blendOption=Linear}
    boundBone {boneName=dyn_skirt_back_01}
    physicsBodyDefinitions [{
      sphereCollisionRadius=0
      boundBone {boneName=dyn_skirt_back_01}
      boxExtents=<5 20 20>
      constraintSetup {angularLimitsMax=<0 30 0>, angularLimitsMin=<0 0 0>, angularTarget=<0 0 0>}
      localJointOffset=<5 0 0>
     }]
   }
   n19 = Anim Dynamics - Bone: dyn_skirt_back_02(ComponentPose=@n18, Alpha=1, ExternalForce="0,0,0") {
    angularDampingOverride=0.999
    bGravityOverrideInSimSpace=true
    bOverrideAngularDamping=true
    gravityScale=0.35
    simulationSpace=RootRelative
    alphaBoolBlend {blendOption=Linear}
    boundBone {boneName=dyn_skirt_back_02}
    physicsBodyDefinitions [{
      sphereCollisionRadius=0
      boundBone {boneName=dyn_skirt_back_02}
      boxExtents=<5 20 20>
      constraintSetup {angularLimitsMax=<0 30 0>, angularLimitsMin=<0 0 0>, angularTarget=<0 0 0>}
      localJointOffset=<5 0 0>
     }]
   }
   n20 = "dyn_knife.rotateY = thigh_l.rotateZ - Bone Driven Controller"!(ComponentPose:Component Space Pose Link Structure=@n19, Alpha:float=1) -> (Pose:Component Space Pose Link Structure) {
    is=title AnimGraphNode_BoneDrivenController
    bAffectTargetRotationY=true
    sourceComponent=RotationZ
    alphaBoolBlend {blendOption=Linear}
    sourceBone {boneName=thigh_l}
    targetBone {boneName=dyn_knife}
   }
   n21 = Anim Dynamics - Bone: dyn_knife(ComponentPose=@n20, Alpha=1, ExternalForce="0,0,0") {
    angularDampingOverride=0.999
    bGravityOverrideInSimSpace=true
    bOverrideAngularDamping=true
    gravityScale=0.35
    simulationSpace=RootRelative
    alphaBoolBlend {blendOption=Linear}
    boundBone {boneName=dyn_knife}
    physicsBodyDefinitions [{
      sphereCollisionRadius=0
      boundBone {boneName=dyn_knife}
      boxExtents=<5 20 20>
      constraintSetup {angularLimitsMax=<0 90 10>, angularLimitsMin=<0 0 -10>, angularTarget=<0 0 0>}
      localJointOffset=<5 0 0>
     }]
   }
   n22 = "Constraint - dyn_spoon (R)"(ComponentPose=@n21, Alpha=1, ConstraintWeights_0=1) {
    alphaBoolBlend {blendOption=Linear}
    boneToModify {boneName=dyn_spoon}
    constraintSetup [{transformType=Rotation, targetBone {boneName=dyn_knife}}]
    constraintWeights [1]
   }
   n23 = Anim Dynamics - Bone: dyn_spoon(ComponentPose=@n22, Alpha=1, ExternalForce="0,0,0") {
    angularDampingOverride=0.999
    bGravityOverrideInSimSpace=true
    bOverrideAngularDamping=true
    gravityScale=0.35
    simulationSpace=RootRelative
    alphaBoolBlend {blendOption=Linear}
    boundBone {boneName=dyn_spoon}
    physicsBodyDefinitions [{
      sphereCollisionRadius=0
      boundBone {boneName=dyn_spoon}
      boxExtents=<5 20 20>
      constraintSetup {angularLimitsMax=<0 90 0>, angularLimitsMin=<0 0 -10>, angularTarget=<0 0 0>}
      localJointOffset=<5 0 0>
     }]
   }
   n24 = Get crouch_alpha() -> (crouch_alpha:double)
   n25 = "Transform (Modify) Bone - Bone: dyn_spoon"(ComponentPose=@n23, Alpha=@n24, Translation="0,0,4", Rotation="0.000,0.000,-10.001", Scale="1,1,1") {
    rotationMode=BMM_Additive
    translationMode=BMM_Additive
    alphaBoolBlend {blendOption=Linear}
    boneToModify {boneName=dyn_spoon}
    rotation {roll=-10.0014}
    translation {z=4}
   }
   n26 = "Transform (Modify) Bone - Bone: dyn_knife"(ComponentPose=@n25, Alpha=@n24, Translation="0,0,0", Rotation="30,0.000,-10.001", Scale="1,1,1") {
    rotationMode=BMM_Additive
    translationMode=BMM_Additive
    alphaBoolBlend {blendOption=Linear}
    boneToModify {boneName=dyn_knife}
    rotation {pitch=30, roll=-10.001}
   }
   n27 = Look At - Bone: dyn_carrots(ComponentPose=@n26, Alpha=1) {
    alphaBoolBlend {blendOption=Linear}
    boneToModify {boneName=dyn_carrots}
    lookAtLocation {x=0}
    lookAtTarget {boneReference {boneName=dyn_skirt_front_02_r}}
    lookAt_Axis {axis=<1 0>}
    lookUp_Axis {axis {x=0, z=1}}
   }
   n28 = Anim Dynamics - Bone: dyn_carrots(ComponentPose=@n27, Alpha=1, ExternalForce="0,0,0") {
    bGravityOverrideInSimSpace=true
    gravityScale=0.35
    simulationSpace=RootRelative
    alphaBoolBlend {blendOption=Linear}
    boundBone {boneName=dyn_carrots}
    physicsBodyDefinitions [{
      sphereCollisionRadius=0
      boundBone {boneName=dyn_carrots}
      boxExtents=<5 20 20>
      constraintSetup {angularLimitsMax=<0 0 10>, angularLimitsMin=<0 -30 -10>, angularTarget=<0 0 0>}
      localJointOffset=<5 0 0>
     }]
   }
   n29 = "Constraint - dyn_shoulder_straps_master_r (PPPPPPPP)"!(
    ComponentPose=@n28
    Alpha=1
    ConstraintWeights_0=0.0096
    ConstraintWeights_1:float=0
    ConstraintWeights_2:float=0
    ConstraintWeights_3:float=0.0067
    ConstraintWeights_4:float=0.5828
    ConstraintWeights_5:float=0.3517
    ConstraintWeights_6:float=0.0136
    ConstraintWeights_7:float=0.0357
   ) -> (Pose) {
    alphaBoolBlend {blendOption=Linear}
    boneToModify {boneName=dyn_shoulder_straps_master_r}
    constraintSetup [
     {targetBone {boneName=upperarm_r}}
     {targetBone {boneName=clavicle_l}}
     {targetBone {boneName=thigh_twist_01_r}}
     {targetBone {boneName=spine_03}}
     {targetBone {boneName=spine_05}}
     {targetBone {boneName=clavicle_r}}
     {targetBone {boneName=neck_02}}
     {targetBone {boneName=neck_01}}
    ]
    constraintWeights [1,1,1,1,1,1,1,1]
   }
   n30 = "Constraint - dyn_shoulder_straps_master_l (PPPPPPPP)"!(
    ComponentPose=@n29
    Alpha=1
    ConstraintWeights_0=0.5035
    ConstraintWeights_1:float=0.0104
    ConstraintWeights_2:float=0.4251
    ConstraintWeights_3:float=0.0117
    ConstraintWeights_4:float=0.0116
    ConstraintWeights_5:float=0.0367
    ConstraintWeights_6:float=0
    ConstraintWeights_7:float=0.0011
   ) -> (Pose) {
    alphaBoolBlend {blendOption=Linear}
    boneToModify {boneName=dyn_shoulder_straps_master_l}
    constraintSetup [
     {targetBone {boneName=clavicle_l}}
     {targetBone {boneName=spine_03}}
     {targetBone {boneName=spine_05}}
     {targetBone {boneName=upperarm_twist_01_l}}
     {targetBone {boneName=neck_02}}
     {targetBone {boneName=neck_01}}
     {targetBone {boneName=thigh_twist_01_l}}
     {targetBone {boneName=upperarm_l}}
    ]
    constraintWeights [1,1,1,1,1,1,1,1]
   }
   n31 = Anim Dynamics - Bone: dyn_shoulder_straps_master_r(ComponentPose=@n30, Alpha=0, ExternalForce="0,0,0") {
    angularDampingOverride=0.999
    bGravityOverrideInSimSpace=true
    bOverrideAngularDamping=true
    gravityScale=0.35
    simulationSpace=RootRelative
    alphaBoolBlend {blendOption=Linear}
    boundBone {boneName=dyn_shoulder_straps_master_r}
    componentAppliedLinearAccClamp=<10 10 10>
    componentLinearAccScale {y=10}
    componentLinearVelScale {y=10}
    physicsBodyDefinitions [{
      sphereCollisionRadius=0
      boundBone {boneName=dyn_shoulder_straps_master_r}
      boxExtents=<10 10 2>
      constraintSetup {angularLimitsMax=<0 20 10>, angularLimitsMin=<0 -30 -10>, angularTarget=<0 0 0>}
      localJointOffset=<2 0 -5>
     }]
   }
   n32 = "Transform (Modify) Bone - Bone: dyn_shoulder_straps_master_r"(ComponentPose=@n31, Alpha=0, Translation="0,0,0", Rotation="11.708,6.261,-25", Scale="1,1,1") {
    rotationMode=BMM_Additive
    alphaBoolBlend {blendOption=Linear}
    boneToModify {boneName=dyn_shoulder_straps_master_r}
    rotation=<rot 11.708 6.261 -25>
   }
   n33 = "Transform (Modify) Bone - Bone: dyn_shoulder_strap_B_01_r"(ComponentPose=@n32, Alpha=1, Translation="0,0,0", Rotation="-10.000,0,5.000", Scale="1,1,1") {
    rotationMode=BMM_Additive
    alphaBoolBlend {blendOption=Linear}
    boneToModify {boneName=dyn_shoulder_strap_B_01_r}
    rotation {pitch=-10, roll=5}
   }
   n34 = "Anim Dynamics - Chain: dyn_shoulder_strap_A_01_r -> dyn_shoulder_strap_A_02_r"(ComponentPose=@n33, Alpha=0.4, ExternalForce="0,0,0") {
    angularDampingOverride=1
    angularSpringConstant=1200
    bAngularSpring=true
    bChain=true
    bGravityOverrideInSimSpace=true
    bOverrideAngularDamping=true
    bUsePlanarLimit=false
    bUseSphericalLimits=true
    gravityScale=0.1
    simulationSpace=RootRelative
    alphaBoolBlend {blendOption=Linear}
    boundBone {boneName=dyn_shoulder_strap_A_01_r}
    chainEnd {boneName=dyn_shoulder_strap_A_02_r}
    componentAppliedLinearAccClamp=<10 10 10>
    physicsBodyDefinitions [{
      collisionType=CustomSphere
      sphereCollisionRadius=2
      boundBone {boneName=dyn_shoulder_strap_A_01_r}
      boxExtents=<1 5 5>
      constraintSetup {angularLimitsMax=<5 30 10>, angularLimitsMin=<-5 -30 -10>, angularTarget=<0 0 0>}
      localJointOffset=<0.8 0 0>
     }]
    sphericalLimits [
     {limitRadius=8, drivingBone {boneName=clavicle_r}, sphereLocalOffset=<-10 1 0>}
     {limitRadius=8, drivingBone {boneName=clavicle_r}, sphereLocalOffset=<-5 2 0>}
     {limitRadius=8, drivingBone {boneName=clavicle_r}, sphereLocalOffset=<-13 0 0>}
    ]
   }
   n35 = "Anim Dynamics - Chain: dyn_shoulder_strap_B_01_r -> dyn_shoulder_strap_B_02_r"(ComponentPose=@n34, Alpha=0.4, ExternalForce="0,0,0") {
    angularDampingOverride=1
    angularSpringConstant=1200
    bAngularSpring=true
    bChain=true
    bGravityOverrideInSimSpace=true
    bOverrideAngularDamping=true
    bUseSphericalLimits=true
    gravityScale=0.1
    simulationSpace=RootRelative
    alphaBoolBlend {blendOption=Linear}
    boundBone {boneName=dyn_shoulder_strap_B_01_r}
    chainEnd {boneName=dyn_shoulder_strap_B_02_r}
    componentAppliedLinearAccClamp=<10 10 10>
    physicsBodyDefinitions [{
      collisionType=CustomSphere
      sphereCollisionRadius=1
      boundBone {boneName=dyn_shoulder_strap_B_01_r}
      boxExtents=<1 5 5>
      constraintSetup {angularLimitsMax=<0 30 10>, angularLimitsMin=<0 -30 0>, angularTarget=<0 0 0>}
      localJointOffset=<1 0 0>
     }]
    sphericalLimits [
     {limitRadius=8, drivingBone {boneName=clavicle_r}, sphereLocalOffset=<-10 1 0>}
     {limitRadius=8, drivingBone {boneName=clavicle_r}, sphereLocalOffset=<-5 2 0>}
     {limitRadius=8, drivingBone {boneName=clavicle_r}, sphereLocalOffset=<-13 0 0>}
    ]
   }
   n36 = Anim Dynamics - Bone: dyn_shoulder_straps_master_l(ComponentPose=@n35, Alpha=0, ExternalForce="0,0,0") {
    angularDampingOverride=0.999
    bGravityOverrideInSimSpace=true
    bOverrideAngularDamping=true
    gravityScale=0.35
    simulationSpace=RootRelative
    alphaBoolBlend {blendOption=Linear}
    boundBone {boneName=dyn_shoulder_straps_master_l}
    componentAppliedLinearAccClamp=<10 10 10>
    componentLinearAccScale {y=10}
    componentLinearVelScale {y=10}
    physicsBodyDefinitions [{
      sphereCollisionRadius=0
      boundBone {boneName=dyn_shoulder_straps_master_l}
      boxExtents=<10 10 2>
      constraintSetup {angularLimitsMax=<0 20 10>, angularLimitsMin=<0 -30 -10>, angularTarget=<0 0 0>}
      localJointOffset=<2 0 -5>
     }]
   }
   n37 = "Transform (Modify) Bone - Bone: dyn_shoulder_straps_master_l"(ComponentPose=@n36, Alpha=0, Translation="0,0,0", Rotation="-5.624,6.419,-11.668", Scale="1,1,1") {
    rotationMode=BMM_Additive
    alphaBoolBlend {blendOption=Linear}
    boneToModify {boneName=dyn_shoulder_straps_master_l}
    rotation=<rot -5.624 6.419 -11.6681>
   }
   n38 = "Transform (Modify) Bone - Bone: dyn_shoulder_strap_B_01_l"(ComponentPose=@n37, Alpha=0, Translation="0,0,0", Rotation="15,-25,-6", Scale="1,1,1") {rotationMode=BMM_Additive, alphaBoolBlend {blendOption=Linear}, boneToModify {boneName=dyn_shoulder_strap_B_01_l}, rotation=<rot 15 -25 -6>}
   n39 = "Anim Dynamics - Chain: dyn_shoulder_strap_A_01_l -> dyn_shoulder_strap_A_02_l"(ComponentPose=@n38, Alpha=0.4, ExternalForce="0,0,0") {
    angularDampingOverride=1
    angularSpringConstant=1200
    bAngularSpring=true
    bChain=true
    bGravityOverrideInSimSpace=true
    bOverrideAngularDamping=true
    bUseSphericalLimits=true
    gravityScale=0.1
    simulationSpace=RootRelative
    alphaBoolBlend {blendOption=Linear}
    boundBone {boneName=dyn_shoulder_strap_A_01_l}
    chainEnd {boneName=dyn_shoulder_strap_A_02_l}
    componentAppliedLinearAccClamp=<10 10 10>
    physicsBodyDefinitions [{
      collisionType=CustomSphere
      sphereCollisionRadius=1
      boundBone {boneName=dyn_shoulder_strap_A_01_l}
      boxExtents=<1 5 5>
      constraintSetup {angularLimitsMax=<5 30 10>, angularLimitsMin=<-5 -30 -10>, angularTarget=<0 0 0>}
      localJointOffset=<1 0 0>
     }]
    sphericalLimits [
     {limitRadius=8, drivingBone {boneName=clavicle_l}, sphereLocalOffset=<10 -1 0>}
     {limitRadius=8, drivingBone {boneName=clavicle_l}, sphereLocalOffset=<5 -2 0>}
     {limitRadius=8, drivingBone {boneName=clavicle_l}, sphereLocalOffset=<13 0 0>}
    ]
   }
   n40 = "Anim Dynamics - Chain: dyn_shoulder_strap_B_01_l -> dyn_shoulder_strap_B_02_l"(ComponentPose=@n39, Alpha=0.4, ExternalForce="0,0,0") {
    angularDampingOverride=1
    angularSpringConstant=1200
    bAngularSpring=true
    bChain=true
    bGravityOverrideInSimSpace=true
    bOverrideAngularDamping=true
    bUseSphericalLimits=true
    gravityScale=0.1
    simulationSpace=RootRelative
    alphaBoolBlend {blendOption=Linear}
    boundBone {boneName=dyn_shoulder_strap_B_01_l}
    chainEnd {boneName=dyn_shoulder_strap_B_02_l}
    componentAppliedLinearAccClamp=<10 10 10>
    physicsBodyDefinitions [{
      collisionType=CustomSphere
      sphereCollisionRadius=1
      boundBone {boneName=dyn_shoulder_strap_B_01_l}
      boxExtents=<1 4 4>
      constraintSetup {angularLimitsMax=<0 30 0>, angularLimitsMin=<0 -30 -10>, angularTarget=<0 0 0>}
      localJointOffset=<0.5 0 0>
     }]
    sphericalLimits [
     {limitRadius=8, drivingBone {boneName=clavicle_l}, sphereLocalOffset=<10 -1 0>}
     {limitRadius=8, drivingBone {boneName=clavicle_l}, sphereLocalOffset=<5 -2 0>}
     {limitRadius=8, drivingBone {boneName=clavicle_l}, sphereLocalOffset=<13 0 0>}
    ]
   }
   n41 = Get crouch_alpha()
   n42 = "Transform (Modify) Bone - Bone: dyn_steak_MASTER"(ComponentPose=@n40, Alpha=@n41, Translation="0,0,0", Rotation="45,0.000,0.000", Scale="1,1,1") {rotationMode=BMM_Additive, alphaBoolBlend {blendOption=Linear}, boneToModify {boneName=dyn_steak_MASTER}, rotation {pitch=45}}
   n43 = "Transform (Modify) Bone - Bone: dyn_steakhook"(ComponentPose=@n42, Alpha=@n41, Translation="0,0,0", Rotation="0,0,-70", Scale="1,1,1") {rotationMode=BMM_Additive, alphaBoolBlend {blendOption=Linear}, boneToModify {boneName=dyn_steakhook}, rotation {roll=-70}}
   n44 = Anim Dynamics - Bone: dyn_steak_01(ComponentPose=@n43, Alpha=@n41, ExternalForce="0,0,0") {
    angularDampingOverride=0.99
    angularSpringConstant=1500
    bAngularSpring=true
    bGravityOverrideInSimSpace=true
    bOverrideAngularDamping=true
    bUsePlanarLimit=false
    gravityScale=0.35
    alphaBoolBlend {blendOption=Linear}
    alphaScaleBiasClamp {bInterpResult=true}
    boundBone {boneName=dyn_steak_01}
    physicsBodyDefinitions [{
      sphereCollisionRadius=0
      boundBone {boneName=dyn_steak_01}
      boxExtents=<5 20 20>
      constraintSetup {angularLimitsMax=<0 20 10>, angularLimitsMin=<0 -20 -10>, angularTarget=<0 0 0>}
      localJointOffset=<5 0 0>
     }]
   }
   n45 = Anim Dynamics - Bone: dyn_steak_02(ComponentPose=@n44, Alpha=@n41, ExternalForce="0,0,0") {
    angularDampingOverride=0.99
    angularSpringConstant=1500
    bAngularSpring=true
    bGravityOverrideInSimSpace=true
    bOverrideAngularDamping=true
    bUsePlanarLimit=false
    gravityScale=0.35
    alphaBoolBlend {blendOption=Linear}
    alphaScaleBiasClamp {bInterpResult=true}
    boundBone {boneName=dyn_steak_02}
    physicsBodyDefinitions [{
      sphereCollisionRadius=0
      boundBone {boneName=dyn_steak_02}
      boxExtents=<5 20 20>
      constraintSetup {angularLimitsMax=<0 20 10>, angularLimitsMin=<0 -20 -10>, angularTarget=<0 0 0>}
      localJointOffset=<5 0 0>
     }]
   }
   n46 = "Anim Dynamics - Chain: dyn_steak_01 -> dyn_steak_02"(ComponentPose=@n45, Alpha=Get crouch_invert_alpha(), ExternalForce="0,0,0") {
    angularDampingOverride=0.99
    angularSpringConstant=1500
    bAngularSpring=true
    bChain=true
    bGravityOverrideInSimSpace=true
    bOverrideAngularDamping=true
    bUsePlanarLimit=false
    gravityScale=0.35
    simulationSpace=RootRelative
    alphaBoolBlend {blendOption=Linear}
    alphaScaleBiasClamp {bInterpResult=true}
    boundBone {boneName=dyn_steak_01}
    chainEnd {boneName=dyn_steak_02}
    physicsBodyDefinitions [{
      sphereCollisionRadius=0
      boundBone {boneName=dyn_steak_01}
      boxExtents=<5 20 20>
      constraintSetup {angularLimitsMax=<0 -10 10>, angularLimitsMin=<0 -30 -10>, angularTarget=<0 0 0>}
      localJointOffset=<5 0 0>
     }]
   }
   Output Pose(Result=Component To Local(ComponentPose=@n46)) {layerGroup=None}
   Get crouch_alpha()
  ]}
 EventGraph {nodes [
   Event BlueprintUpdateAnimation()
   n1 = GetPartSkeletalMeshComponent()
   n2 = BreakRotator(InRot=BreakTransform(InTransform=GetDeltaTransformFromRefPose(self=@n1, BoneName=thigh_l, BaseName=None)).Rotation)
   > Set crouch_alpha(crouch_alpha:double=float - float(B=MapRangeClamped(Value=@n2.Yaw, InRangeA=-100, InRangeB=-50)) {is=title K2Node_CallFunction}) -> (Output_Get:double)
   n3 = Event OnEndCrouching()
   n4 = "To Float (Boolean)"(InBool=Get bIsCrouching())
   Set crouch_alpha(execute=[@n3.then, @n5.then], crouch_alpha=@n4)
   > Set crouch_invert_alpha(crouch_invert_alpha:double=float - float(B=@n4)) -> (Output_Get:double)
   n5 = Event OnBeginCrouching()
  ]}
]
variables [
 crouch_alpha {type=number}
 crouch_invert_alpha {type=number, default=1}
]
