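# Physics asset for the cloth on the Dumpling Man Beef male medium character body, evidently serving as collision for cloth simulation.
# It covers only the upper body, with one capsule each on spine_05, neck_01 and the left and right clavicles.
# Constraints link the neck and both clavicles to spine_05 with all swing and twist motion free, so they impose no angular limits.
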PhysicsAsset /BRCosmetics/Characters/Player/Male/Medium/Bodies/M_MED_Dumpling_Man_Beef/Meshes/M_MED_Dumpling_Man_Beef_Cloth_Physics
bodies {
 clavicle_l [{shapeType=Capsule, length=15.097, radius=12.5917, center=<4.147 3.4184 -1.0715>, rotation=<rot 67.897 0 0>}]
 clavicle_r [{shapeType=Capsule, length=15.097, radius=12.5917, center=<-2.8102 -3.901 0.3177>, rotation=<rot 66.9699 0.0001 -179.9999>}]
 neck_01 [{shapeType=Capsule, length=12.75, radius=8.9386, rotation=<rot 90 0 3.2439>}]
 spine_05 [{shapeType=Capsule, length=11.0148, radius=12.3637, center=<0 0.5 0>}]
}
constraints [
 {
  bone1Name=clavicle_l
  bone2Name=spine_05
  swing1LimitDegrees=45
  swing1Motion=Free
  swing2LimitDegrees=45
  swing2Motion=Free
  twistLimitDegrees=45
  twistMotion=Free
 }
 {
  bone1Name=clavicle_r
  bone2Name=spine_05
  swing1LimitDegrees=45
  swing1Motion=Free
  swing2LimitDegrees=45
  swing2Motion=Free
  twistLimitDegrees=45
  twistMotion=Free
 }
 {
  bone1Name=neck_01
  bone2Name=spine_05
  swing1LimitDegrees=45
  swing1Motion=Free
  swing2LimitDegrees=45
  swing2Motion=Free
  twistLimitDegrees=45
  twistMotion=Free
 }
]
